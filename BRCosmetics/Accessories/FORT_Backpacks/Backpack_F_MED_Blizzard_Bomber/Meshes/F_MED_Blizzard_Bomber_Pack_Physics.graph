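# Physics asset for the Blizzard Bomber backpack cosmetic, providing bodies and joints for its dangling rope, mask strap and clip.
# A compound capsule-and-box body on the pack bone anchors all three attachments.
# The six-capsule rope has a fully free root joint and tighter limits down the chain, while the three-capsule strap uses narrow limits throughout.
# The box clip locks twist and one swing axis, giving hinge-like motion.
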PhysicsAsset /BRCosmetics/Accessories/FORT_Backpacks/Backpack_F_MED_Blizzard_Bomber/Meshes/F_MED_Blizzard_Bomber_Pack_Physics
bodies {
 dyn_clip_01_joint [{shapeType=Box, extentX=1.4437, extentY=2.6089, extentZ=0.6961, shapeName=dyn_clip_01_joint_box, center=<1.7969 0 0>}]
 dyn_mask_strap_01_joint [{
   shapeType=Capsule
   length=0.2912
   radius=0.5
   shapeName=dyn_mask_strap_01_joint_capsule
   center=<0.6149 0.0172 -0.0277>
   rotation=<rot -0.7658 178.098 -81.3799>
  }]
 dyn_mask_strap_02_joint [{
   shapeType=Capsule
   length=0
   radius=0.5
   shapeName=dyn_mask_strap_02_joint_capsule
   center=<0.6715 0.0008 0.0315>
   rotation=<rot 0 5.3844 -86.4559>
  }]
 dyn_mask_strap_03_joint [{
   shapeType=Capsule
   length=0
   radius=0.5
   shapeName=dyn_mask_strap_03_joint_capsule
   center=<0.6228 -0.0347 0.0402>
   rotation=<rot 0.0746 -2.1456 -92.0557>
  }]
 dyn_rope_01_joint [{
   shapeType=Capsule
   length=1.9123
   radius=1.8572
   shapeName=dyn_rope_01_joint_capsule
   center=<2.5833 -0.3763 0.176>
   rotation=<rot -17.7835 -82.3812 -1.8204>
  }]
 dyn_rope_02_joint [{
   shapeType=Capsule
   length=3.3976
   radius=1.8904
   shapeName=dyn_rope_02_joint_capsule
   center=<1.879 -0.6734 0.3255>
   rotation=<rot -18.829 -89.3305 0.8331>
  }]
 dyn_rope_03_joint [{
   shapeType=Capsule
   length=3.8209
   radius=1.7268
   shapeName=dyn_rope_03_joint_capsule
   center=<2.2203 -0.5422 0.0612>
   rotation=<rot -15.5291 -90.0517 -0.3734>
  }]
 dyn_rope_04_joint [{
   shapeType=Capsule
   length=3.6067
   radius=1.679
   shapeName=dyn_rope_04_joint_capsule
   center=<1.8614 -0.2361 0.2936>
   rotation=<rot -13.6658 -89.9997 -0.1017>
  }]
 dyn_rope_05_joint [{
   shapeType=Capsule
   length=2.9465
   radius=1.8111
   shapeName=dyn_rope_05_joint_capsule
   center=<2.2359 -0.087 0.2348>
   rotation=<rot -16.5213 -89.9998 -0.1256>
  }]
 dyn_rope_06_joint [{
   shapeType=Capsule
   length=0.9203
   radius=1.8798
   shapeName=dyn_rope_06_joint_capsule
   center=<2.2155 0.0348 0.1071>
   rotation=<rot -11.1165 -87.8737 -0.7782>
  }]
 pack [
  {
   shapeType=Capsule
   length=23.0505
   radius=5.2529
   shapeName=pack_capsule
   center=<0.7011 -5.965 -1.5587>
   rotation=<rot 12.0802 95.9804 -45.458>
  }
  {shapeType=Capsule, length=15, radius=6.4506, shapeName=pack_capsule, center=<2.5281 -10.7174 -2.7059>, rotation=<rot -47.014 0 0>}
  {shapeType=Capsule, length=17.9235, radius=6.1112, shapeName=pack_capsule, center=<1.8443 -6.1795 -5.1021>, rotation=<rot -46.6805 0 0>}
  {
   shapeType=Box
   extentX=22.7546
   extentY=3.7577
   extentZ=2.4018
   shapeName=pack_box
   center=<-2.1697 -6.4419 2.3461>
   rotation=<rot 42.7672 -0.0004 -0.0006>
  }
  {shapeType=Box, extentX=18.8806, extentY=5.7769, extentZ=33.0106, shapeName=pack_box, center=<-4.2225 2.4768 -20.7622>}
 ]
}
constraints [
 {bone1Name=dyn_mask_strap_02_joint, bone2Name=dyn_mask_strap_01_joint, swing1LimitDegrees=15, swing2LimitDegrees=15, twistLimitDegrees=3}
 {bone1Name=dyn_mask_strap_03_joint, bone2Name=dyn_mask_strap_02_joint, swing1LimitDegrees=20, swing2LimitDegrees=20, twistLimitDegrees=3}
 {bone1Name=dyn_mask_strap_01_joint, bone2Name=pack, swing1LimitDegrees=15, swing2LimitDegrees=15, twistLimitDegrees=3}
 {
  bone1Name=dyn_clip_01_joint
  bone2Name=pack
  swing1LimitDegrees=0
  swing1Motion=Locked
  swing2LimitDegrees=180
  twistLimitDegrees=0
  twistMotion=Locked
 }
 {
  bone1Name=dyn_rope_01_joint
  bone2Name=pack
  swing1LimitDegrees=45
  swing1Motion=Free
  swing2LimitDegrees=45
  swing2Motion=Free
  twistLimitDegrees=45
  twistMotion=Free
 }
 {bone1Name=dyn_rope_02_joint, bone2Name=dyn_rope_01_joint, swing1LimitDegrees=20, swing2LimitDegrees=20, twistLimitDegrees=5}
 {bone1Name=dyn_rope_03_joint, bone2Name=dyn_rope_02_joint, swing1LimitDegrees=10, swing2LimitDegrees=15, twistLimitDegrees=5}
 {bone1Name=dyn_rope_04_joint, bone2Name=dyn_rope_03_joint, swing1LimitDegrees=10, swing2LimitDegrees=15, twistLimitDegrees=5}
 {bone1Name=dyn_rope_05_joint, bone2Name=dyn_rope_04_joint, swing1LimitDegrees=10, swing2LimitDegrees=15, twistLimitDegrees=5}
 {bone1Name=dyn_rope_06_joint, bone2Name=dyn_rope_05_joint, swing1LimitDegrees=10, swing2LimitDegrees=15, twistLimitDegrees=5}
]
