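# Animation Blueprint for the Blizzard Bomber backpack cosmetic, which positions the pack and adds physics motion.
# An additive root offset feeds a base-bone-space RigidBody node, then a fixed additive offset on the pack bone.
# A deprecated gravity-override function drives simulation gravity each update, while jump and fall events adjust linear acceleration scale.
# Skydive and parachute events swap root offsets, and updates restore defaults when neither falling nor jumping.
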AnimBlueprint /BRCosmetics/Accessories/FORT_Backpacks/Backpack_F_MED_Blizzard_Bomber/Meshes/F_MED_Blizzard_Bomber_Pack_AnimBP
target_skeleton=&/F_MED_Blizzard_Bomber_Pack_Skeleton
graphs [
 AnimGraph {nodes [
   n1 = "Transform (Modify) Bone - Bone: root"(
    ComponentPose=Local To Component(LocalPose=Input Pose())
    Translation=Get translationBlizzardBomber() -> (translationBlizzardBomber:Vector)
    Rotation=Get rotatorBlizzardBomber() -> (rotatorBlizzardBomber:Rotator)
   ) {
    rotationMode=BMM_Additive
    translationMode=BMM_Additive
    alphaBoolBlend {blendOption=Linear}
    boneToModify {boneName=root}
    rotation=<rot 9.148 5.464 5.404>
    translation=<2.06 3 2.89>
   }
   n2 = RigidBody!(
    ComponentPose=@n1
    Alpha
    OverrideWorldGravity:Vector=Get skyDiving() -> (skyDiving:Vector)
    ExternalForce
    ComponentLinearAccScale:Vector=Get falling() -> (falling:Vector)
   ) -> (Pose) {
    bOverrideWorldGravity=true
    overridePhysicsAsset=&/F_MED_Blizzard_Bomber_Pack_Physics
    simulationSpace=BaseBoneSpace
    alphaBoolBlend {blendOption=Linear}
    baseBoneRef {boneName=root}
    componentLinearAccScale=<0.5 0.15 0.5>
    overrideWorldGravity {z=-980}
   }
   n3 = "Transform (Modify) Bone - Bone: pack"(ComponentPose=@n2, Translation="-1.5,-0.5,-6", Rotation="0,0,0") {translationMode=BMM_Additive, alphaBoolBlend {blendOption=Linear}, boneToModify {boneName=pack}, translation=<-1.5 -0.5 -6>}
   Output Pose(Result=Component To Local(ComponentPose=@n3)) {layerGroup=None}
  ]}
 EventGraph {nodes [
   Event BlueprintUpdateAnimation()
   > n1 = Gravity Override State Function DEPRECATED(CharacterPartAnimInstance=Self-Reference(), GameTimeInSeconds=GetGameTimeInSeconds(), GravityOverrideParams=Get Gravity Override Params())
   > Set skyDiving(skyDiving:Vector=@n1) -> (Output_Get:Vector)
   > n2 = Branch(Condition=OR Boolean(A=Get bIsFalling(), B=Get bIsJumping()))
   Set falling(execute=@n2.else, falling:Vector="0.5,0.15,0.5") -> (Output_Get:Vector)
   > Set translationBlizzardBomber(translationBlizzardBomber:Vector="2.06,3,2.89") -> (Output_Get:Vector)
   > Set rotatorBlizzardBomber(rotatorBlizzardBomber:Rotator="9.148,5.464,5.404") -> (Output_Get:Rotator)
   n3 = Event OnEndJumping()
   RetriggerableDelay(execute=[@n3.then, @n5.then], Duration=0)
   > Set falling(falling="0.5,0.15,0.5")
   n4 = Event OnBeginJumping()
   RetriggerableDelay(execute=[@n4.then, @n6.then], Duration=0)
   > Set falling(falling="0.5,0.5,0.5")
   n5 = Event OnEndFalling()
   n6 = Event OnBeginFalling()
   Event OnBeginSkydiving()
   > Set translationBlizzardBomber(translationBlizzardBomber="0,3, 0")
   > Set rotatorBlizzardBomber(rotatorBlizzardBomber="0,-15, 0")
   Event OnEndSkydiving()
   > Set translationBlizzardBomber(translationBlizzardBomber="2.06,3,2.89")
   > Set rotatorBlizzardBomber(rotatorBlizzardBomber="9.148,5.464,5.404")
   Event OnParachuteOpened()
   > Set translationBlizzardBomber(translationBlizzardBomber="0,3, 0")
   > Set rotatorBlizzardBomber(rotatorBlizzardBomber="0, 0, 0")
   Event OnParachuteClosed()
   > Set translationBlizzardBomber(translationBlizzardBomber="0,3, 0")
   > Set rotatorBlizzardBomber(rotatorBlizzardBomber="0,-15, 0")
   TryGetPawnOwner()
   Event OnWeaponChanged()
  ]}
]
variables [
 falling {type=Vector}
 Gravity Override Params {type=GravityOverrideParamsStruct}
 rotateRopeJoints {type=Rotator}
 rotatorBlizzardBomber {type=Rotator}
 skyDiving {type=Vector}
 translationBlizzardBomber {type=Vector}
]
